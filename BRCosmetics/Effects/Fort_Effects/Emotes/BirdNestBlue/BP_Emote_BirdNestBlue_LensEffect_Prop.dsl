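; Emote prop actor for BirdNestBlue that applies a post-process lens effect for the player viewing the emoting pawn.
; On BeginPlay outside the frontend, it waits briefly, then casts its attach parent to a FortPawn.
; If that pawn is locally viewed, a timeline's "v" and "ca" tracks drive the PostProcess component's settings each update.
; Otherwise the actor destroys itself, keeping the effect off other players' screens.
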(event EventBeginPlay
  (if (not (Reactive|Frontend|IsCosmeticObjectDisplayedinFrontendContext 0))
    (Utilities|FlowControl|Delay 0.1)
    (bind _asfort_pawn (Utilities|Casting|CastToFortPawn (Actor|GetAttachParentActor self)))
    (if (Pawn|IsLocallyViewed _asfort_pawn)
      (bind (_direction _v _ca) (Timeline|Timeline PlayFromStart)
        (:Update
          (Variables|/Script/Engine.PostProcessComponent|SetSettings (Utilities|Struct|MakePostProcessSettings _ca _v) (Variables|Default|GetPostProcess))))
      (else
        (Actor|DestroyActor)))))

(event EventDestroyed
  (Actor|DestroyActor))

(fn ConstructionScript)
