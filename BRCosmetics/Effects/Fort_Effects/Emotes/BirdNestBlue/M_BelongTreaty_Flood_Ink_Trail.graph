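# Translucent, unlit, two-sided Niagara sprite and mesh particle material for an ink trail in the BirdNestBlue emote.
# Particle color feeds MF_FX_EmissiveAdjustments for emissive.
# Opacity is particle alpha times a channel-selectable trail texture with scalable UVs, shaped by a dynamic AlphaPower exponent and an optional radial sphere mask.
# A shading path switch applies depth fade only on default and deferred paths, skipping it on forward and mobile.
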Material /BRCosmetics/Effects/Fort_Effects/Emotes/BirdNestBlue/M_BelongTreaty_Flood_Ink_Trail
graph {
 properties {
  bUsedWithNiagaraMeshParticles=true
  bUsedWithNiagaraSprites=true
  blendMode=BLEND_Translucent
  pixelDepthOffsetMode=PDOM_Legacy
  shadingModel=MSM_Unlit
  twoSided=true
  usedShadingModels=Unlit
 }
 nodes [
  n1 = ParticleColor()
  MaterialFunctionCall(Emissive=@n1.RGB) {materialFunction=&MF_FX_EmissiveAdjustments}
  n2 = TextureCoordinate()
  n3 = ComponentMask(None=VectorParameter() {parameterName=SpriteUVScale, defaultValue=<rgba 1 1 0 0>}.RGB) {g=true, r=true}
  n4 = StaticComponentMaskParameter(None=TextureSampleParameter2D(UVs=Multiply(A=@n2, B=@n3)) {parameterName=Sprite Tex, texture=&/T_InkTrail}.RGB) {defaultB=true, parameterName=Sprite mask}
  n5 = DynamicParameter() {paramNames [AlphaPower,Param2,Param3,Param4]}
  n6 = SphereMask(A=TextureCoordinate(), B=Constant2Vector() {g=0.5, r=0.5}) {attenuationRadius=0.5, hardnessPercent=0}
  n7 = ScalarParameter() {defaultValue=1, parameterName=Use sphere mask}
  n8 = Multiply(A=@n1.A, B=Multiply(A=Power(Base=@n4, Exp=@n5.AlphaPower), B=LinearInterpolate(B=@n6, Alpha=@n7) {constA=1}))
  n9 = DepthFade(Opacity=@n8, FadeDistance=ScalarParameter() {parameterName=DepthFade}) {fadeDistanceDefault=5}
  ShadingPathSwitch(Default=@n9, Deferred=@n9, Forward=@n8, Mobile=@n8)
 ]
}
