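# Unlit, two-sided translucent air-flow material for Niagara mesh particles in the BirdNestBlue emote.
# Emissive blends a stepped UV gradient with panned blurred noise, tinted by particle color.
# Opacity multiplies particle alpha, a toon-ribbon flipbook frame, a swirl-noise HeightLerp dissolve and view-angle fade, with dynamic parameters driving frame and dissolve.
# A shading-path switch skips depth fade on forward and mobile; the Default Tex branch feeds nothing.
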Material /BRCosmetics/Effects/Fort_Effects/Emotes/BirdNestBlue/M_BirdNestBlue_AirFlow
graph {
 properties {
  bUsedWithNiagaraMeshParticles=true
  blendMode=BLEND_Translucent
  pixelDepthOffsetMode=PDOM_Legacy
  shadingModel=MSM_Unlit
  twoSided=true
  usedShadingModels=Unlit
 }
 nodes [
  n1 = SmoothStep(Min=Constant() {r=0.49}, Value=ComponentMask(None=TextureCoordinate()) {g=true})
  n2 = Saturate(None=Multiply(A=LinearInterpolate(Alpha=@n1) {constA=0.03}) {constB=5})
  n3 = TextureCoordinate()
  n4 = ComponentMask(None=VectorParameter() {parameterName=Color noise UV Sacle, defaultValue=<rgba 0.3 0.5 0 0>}.RGB) {g=true, r=true}
  n5 = VectorParameter() {parameterName=Color noise UV offset, defaultValue=<rgba 0.5 0.5 0 1>}
  n6 = Panner(Coordinate=Add(A=Multiply(A=@n3, B=@n4), B=ComponentMask(None=@n5.RGB) {g=true, r=true}))
  n7 = Multiply(B=Power(
    Base=TextureSample(UVs=@n6) {samplerType=SAMPLERTYPE_Grayscale, texture=&LowResBlurredNoise}.R
    Exp=ScalarParameter() {defaultValue=1.3, parameterName=Color noise Power}
   ) {constExponent=1.3}) {constA=1}
  n8 = ScalarParameter() {parameterName=Use detail color mask}
  n9 = ParticleColor()
  MaterialFunctionCall(Emissive=Multiply(A=LinearInterpolate(A=@n2, B=@n7, Alpha=@n8), B=@n9.RGB)) {materialFunction=&MF_FX_EmissiveAdjustments}
  n10 = Constant2Vector() {g=2, r=2}
  n11 = MaterialFunctionCall!(
   U_Images=@n10.R
   V_Images=@n10.R
   SubUV Texture RGB=TextureObjectParameter() {parameterName=A, texture=&/T_ToonStyleRibbons}
   Time Speed
   Manual Time Control=StaticBool() {value=true}
   Time=DynamicParameter() {paramNames [Panner,Param2,Param3,Sub]}.Sub
   UVs
  ) {materialFunction=/Game/Effects/Fort_Effects/Materials/_Functions/MF_SubUV}
  n12 = DynamicParameter() {paramNames [Panner,Param2,Param3,Sub]}
  n13 = TextureCoordinate()
  n14 = Constant2Vector()
  n15 = Multiply(A=TextureSample(UVs=Add(A=@n13, B=@n14)) {samplerType=SAMPLERTYPE_LinearGrayscale, texture=/Game/Global/Textures/Default/Noise/T_SwirlNoise_N_height}.R)
  n16 = ScalarParameter() {defaultValue=0.5, parameterName="Contrast (S)"}
  n17 = Multiply(
   A=Multiply(
    A=Multiply(A=@n9.A, B=@n11.RGB)
    B=MaterialFunctionCall(Transition Phase=@n12.Panner, Height Texture=@n15, Contrast=@n16) {materialFunction=&HeightLerp}.Alpha
   )
   B=MaterialFunctionCall("FadeSensitivity (optional)"=ScalarParameter() {parameterName=Edge fade}) {materialFunction=&FadeBasedOnViewAngle}.Result
  )
  n18 = DepthFade(Opacity=@n17) {fadeDistanceDefault=30}
  ShadingPathSwitch(Default=@n18, Deferred=@n18, Forward=@n17, Mobile=@n17)
  n19 = TextureCoordinate()
  n20 = ComponentMask(None=VectorParameter() {parameterName=UV Scale, defaultValue=<rgba 1 1 0 0>}.RGB) {g=true, r=true}
  StaticComponentMaskParameter(None=TextureSampleParameter2D(UVs=Multiply(A=@n19, B=@n20)) {parameterName=Default Tex, texture=/Game/Effects/Fort_Effects/Emotes/Energize_Max/Burst/T_Emote_Energize_airflow}.RGB) {defaultR=true, parameterName=Component mask}
  TextureCoordinate()
  TextureCoordinate() {uTiling=0.3, vTiling=0.5}
 ]
}
