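# Translucent, unlit, two-sided material for Niagara mesh particles that draws a dark ink stroke in the BirdNestBlue emote effect.
# Particle color times a black-default tint drives emissive, while SplineThicken sets stroke width from a U-axis edge mask and dynamic Width.
# Dynamic parameters also slide the clamped ink-trail texture, with optional sine debug animation, and drive a ChooseStroke blend.
# Opacity comes from that sample after gray adjustments, multiplied by an edge fade along U.
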Material /BRCosmetics/Effects/Fort_Effects/Emotes/BirdNestBlue/M_BirdNestBlue_DarkInkStroke
graph {
 properties {
  bUsedWithNiagaraMeshParticles=true
  blendMode=BLEND_Translucent
  pixelDepthOffsetMode=PDOM_Legacy
  shadingModel=MSM_Unlit
  twoSided=true
  usedShadingModels=Unlit
 }
 nodes [
  n1 = ParticleColor()
  n2 = VectorParameter() {parameterName=Color, defaultValue=<rgba 0 0 0 1>}
  MaterialFunctionCall(Emissive=Divide(A=Multiply(A=@n1.RGB, B=@n2.RGB))) {materialFunction=&MF_FX_EmissiveAdjustments}
  n3 = VectorParameter() {parameterName=WidthLerp, defaultValue=<rgba 0 5 0 1>}
  n4 = MaterialFunctionCall(Float2=TextureCoordinate()) {materialFunction=&BreakOutFloat2Components}
  n5 = VectorParameter() {group=Width, parameterName=Width_EdgeMask_Control, channelNames {a="", b=Power, g=Multiply, r=Offset}, defaultValue=<rgba 0 5 1 1>}
  n6 = DynamicParameter() {defaultValue=<rgba 0 1 0 1>, paramNames [StrokeTextureMove, DisableDebug, ChooseStroke, Width]}
  n7 = Multiply(
   A=LinearInterpolate(
    A=@n3.R
    B=@n3.R
    Alpha=MaterialFunctionCall!(UorV=@n4.R, EdgeControl=@n5) {materialFunction=/Game/Effects/Niagara/MasterMaterials/Functions/Common/FXCommon_EdgeMask_Simple}.Mask
   )
   B=@n6.Width
  )
  MaterialFunctionCall(WidthBase=@n7, WidthTip=@n7, Expand U or V UV Channel=StaticBool()) {materialFunction=&SplineThicken}
  n8 = TextureObjectParameter() {
   group=StrokeTexture
   parameterName=StrokeTexture_Texture
   sortPriority=0
   texture=&/T_InkTrail
  }
  n9 = VectorParameter() {parameterName=StrokeTextureMove, channelNames {a="", b=Lp, g=B, r=A}, defaultValue=<rgba -1 2 0.5 1>}
  n10 = DynamicParameter() {defaultValue=<rgba 0 1 0 1>, paramNames [StrokeTextureMove, DisableDebug, ChooseStroke, Width]}
  n11 = Add(
   A=Multiply(
    A=@n10.DisableDebug
    B=MaterialFunctionCall(Speed=@n9.Lp) {materialFunction=&DebugTimeSine}.Result
   )
   B=@n10.StrokeTextureMove
  )
  n12 = VectorParameter() {group=StrokeTexture, parameterName=StrokeTexture_UVControl, channelNames {a=VPan, b=UPan, g=VScale, r=UScale}, defaultValue=<rgba 0.5 1 0 0>}
  n13 = MaterialFunctionCall(
   Texture=@n8
   AdditionalUVOffset=AppendVector(A=LinearInterpolate(A=@n9.A, B=@n9.A, Alpha=@n11), B=Constant())
   UVControl_RGB=@n12
   UVControl_A=@n12
  ) {materialFunction=&FXCommon_TextureSample}
  n14 = Divide(A=@n6.ChooseStroke)
  n15 = Multiply(
   A=MaterialFunctionCall!(TexCoord=@n13.UV, ClampChannel=Constant2Vector() {g=1, r=1}) {materialFunction=/Game/Effects/Fort_Effects/MaterialFunctions/MF_FX_TilingMethod}.Result
   B=MaterialFunctionCall!(1=@n13.R, 2=@n13.R, 3=@n13.R, 4, A=Saturate(None=@n14)) {materialFunction=/Engine/Functions/Engine_MaterialFunctions02/Lerp_Multiple_Float}.Lerp 3 Inputs
  )
  n16 = ScalarParameter() {group=StrokeAlpha, parameterName=StrokeAlpha_GrayAdjust_Add}
  n17 = Saturate(None=Multiply(A=Add(A=@n15, B=@n16), B=ScalarParameter() {defaultValue=2, group=StrokeAlpha, parameterName=StrokeAlpha_GrayAdjust_Mul}))
  n18 = ScalarParameter() {defaultValue=1, group=StrokeAlpha, parameterName=StrokeAlpha_GrayAdjust_Power}
  n19 = VectorParameter() {group=Fade, parameterName=Fade_EdgeMask_InvertControl, channelNames {a="", b=Power, g=Multiply, r=Offset}, defaultValue=<rgba 0 20 1.5 1>}
  n20 = VectorParameter() {group=Fade, parameterName=Fade_EdgeMask_Control, channelNames {a="", b=Power, g=Multiply, r=Offset}, defaultValue=<rgba 0 4 2 1>}
  Multiply(
   A=Saturate(None=Power(Base=@n17, Exp=@n18))
   B=MaterialFunctionCall!(UorV=@n4.R, Invert_EdgeControl=@n19, NoInvert_EdgeControl=@n20) {materialFunction=/Game/Effects/Niagara/MasterMaterials/Functions/Common/FXCommon_EdgeMask_Ex}.Mask
  )
 ]
}
