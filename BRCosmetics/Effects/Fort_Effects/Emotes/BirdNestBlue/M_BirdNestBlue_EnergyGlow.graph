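# Translucent, unlit, two-sided Niagara sprite material for the BirdNestBlue emote's energy glow, tinted by particle color.
# A plasma texture samples UVs tapered horizontally along V and distorted by scrolling Perlin noise weighted by a vertical gradient.
# An inverted sphere mask is subtracted to fade the edges, then contrast and particle alpha shape the result.
# A DPTime reroute uses the dynamic parameter time input, falling back to game time at zero.
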Material /BRCosmetics/Effects/Fort_Effects/Emotes/BirdNestBlue/M_BirdNestBlue_EnergyGlow
graph {
 properties {
  bUsedWithNiagaraSprites=true
  blendMode=BLEND_Translucent
  pixelDepthOffsetMode=PDOM_Legacy
  shadingModel=MSM_Unlit
  twoSided=true
  usedShadingModels=Unlit
 }
 nodes [
  Constant3Vector()
  n1 = ParticleColor()
  MaterialFunctionCall(Emissive=Divide(A=@n1.RGB)) {materialFunction=&MF_FX_EmissiveAdjustments}
  n2 = TextureObjectParameter() {
   group=Texture
   parameterName=Texture_Texture
   samplerType=SAMPLERTYPE_LinearGrayscale
   sortPriority=0
   texture=/Game/Effects/Fort_Effects/Emotes/UltraLight/Texture/T_Plasma_A_B_Greyscale
  }
  n3 = VectorParameter() {group=Texture, parameterName=Texture_UVControl, channelNames {a=VPan, b=UPan, g=VScale, r=UScale}, defaultValue=<rgba 1 0.13 0 2.4>}
  n4 = TextureCoordinate()
  n5 = Add(
   A=OneMinus(None=MaterialFunctionCall(Float2=@n4) {materialFunction=&BreakOutFloat2Components}.G)
   B=ScalarParameter() {group=TotalUVNoise, parameterName=TotalUVNoise_GrayAdjust_Add}
  )
  n6 = ScalarParameter() {defaultValue=1, group=TotalUVNoise, parameterName=TotalUVNoise_GrayAdjust_Mul}
  n7 = Power(
   Base=Saturate(None=Multiply(A=@n5, B=@n6))
   Exp=ScalarParameter() {defaultValue=2, group=TotalUVNoise, parameterName=TotalUVNoise_GrayAdjust_Power}
  )
  n8 = TextureObjectParameter() {
   group=TotalUVNoise
   parameterName=TotalUVNoise_Texture
   samplerType=SAMPLERTYPE_Normal
   sortPriority=0
   texture=/Game/Effects/Fort_Effects/Emotes/BirdNestBlue/T_LowResSimplePerlin
  }
  n9 = VectorParameter() {
   group=TotalUVNoise
   parameterName=TotalUVNoise_UVControl
   channelNames {a=VPan, b=UPan, g=VScale, r=UScale}
   defaultValue=<rgba 0.02 0.4 0 1.6>
  }
  n10 = MaterialFunctionCall(
   Texture=@n8
   UVControl_RGB=@n9
   UVControl_A=@n9
   "[Override]TimeInput"=DynamicParameter() {defaultValue=<rgba 0 0 1 1>, paramNames [TimeInput,Random,Param3,Param4]}.TimeInput
  ) {materialFunction=&FXCommon_TextureSample}
  n11 = VectorParameter() {
   group=TotalUVNoise
   parameterName=TotalUVNoise_NoiseControl
   sortPriority=60
   channelNames {a=VStrength, b=UStrength, g=VOffset, r=UOffset}
   defaultValue=<rgba 0 0 1 0>
  }
  n12 = TextureCoordinate()
  n13 = Power(Base=OneMinus(None=MaterialFunctionCall(Float2=@n12) {materialFunction=&BreakOutFloat2Components}.G)) {constExponent=1}
  n14 = MaterialFunctionCall(UVs=@n12, Texture Scale=AppendVector(A=OneMinus(None=@n13), B=Constant() {r=1})) {materialFunction=&ScaleUVsByCenter}
  n15 = VectorParameter() {group=MainShape, parameterName=TotalUVOffsetAndLerpScaleU, channelNames {a="", b="", g=VOffset, r=UOffset}}
  n16 = AppendVector(A=@n15.UOffset, B=@n15.UOffset)
  n17 = VectorParameter() {group=MainShape, parameterName=TotalUVControl, channelNames {a=VPan, b=UPan, g=VScale, r=UScale}, defaultValue=<rgba 1 1 0 0>}
  n18 = Add(
   A=Multiply(
    A=Saturate(None=@n7)
    B=MaterialFunctionCall!(NoiseValue=@n10.RG, Control_RGB=@n11, Control_A=@n11) {materialFunction=/Game/Effects/Niagara/MasterMaterials/Functions/Common/FXCommon_NoiseControl}.NoiseValue
   )
   B=MaterialFunctionCall!(UVInput=Add(A=@n14.UVs, B=@n16), TimeInput=NamedRerouteUsage(), Control_RGB=@n17, Control_A=@n17, ScaleUVByCenter=StaticBool() {value=true}) {materialFunction=/Game/Effects/Niagara/MasterMaterials/Functions/Common/FXCommon_QuickUVControl_V2}.UV
  )
  n19 = MaterialFunctionCall(Texture=@n2, UVControl_RGB=@n3, UVControl_A=@n3, "[Override]TimeInput"=NamedRerouteUsage(), "[Override]UVInput"=@n18) {materialFunction=&FXCommon_TextureSample}
  n20 = Constant() {r=0.5}
  n21 = OneMinus(None=Power(Base=SphereMask(A=@n18, B=@n20) {attenuationRadius=0.5, hardnessPercent=0}) {constExponent=0.3})
  Multiply(
   A=@n1.A
   B=MaterialFunctionCall(In=Subtract(A=@n19.R, B=@n21), Contrast=Constant() {r=2}) {materialFunction=&CheapContrast}.Result
  )
  n22 = DynamicParameter() {defaultValue=<rgba 0 0 1 1>, paramNames [TimeInput,Random,Param3,Param4]}
  n23 = Add(A=@n22.TimeInput, B=@n22.TimeInput)
  n24 = Time() {bOverride_Period=true, period=6666}
  NamedRerouteDeclaration(None=If(A=@n23, "A > B"=@n23, "A == B"=@n24, "A < B"=@n24)) {name=DPTime, nodeColor=<rgba 1 0.7266 0 1>}
  TextureCoordinate()
 ]
}
