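# Translucent unlit Niagara sprite material for the BirdNestBlue emote that fakes a toon light by brightening scene color with particle color.
# A soft world-space sphere mask behind the translucency sets opacity, with radius from dynamic parameters and all three center axes wired to PosX.
# A subsurface mask, read from GBuffer data chosen per quality level, limits the light to matching pixels.
# Forward and mobile paths output zero, disabling the effect.
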Material /BRCosmetics/Effects/Fort_Effects/Emotes/BirdNestBlue/M_BirdNestBlue_ToonLight
graph {
 properties {bUsedWithNiagaraSprites=true, blendMode=BLEND_Translucent, pixelDepthOffsetMode=PDOM_Legacy, shadingModel=MSM_Unlit, usedShadingModels=Unlit}
 nodes [
  n1 = ParticleColor()
  n2 = Constant()
  ShadingPathSwitch(Default=Multiply(A=SceneColor!(UVs).RGB, B=@n1.RGB) {constB=20}, Forward=@n2, Mobile=@n2)
  n3 = SceneTexture() {sceneTextureId=PPI_SubsurfaceColor}
  n4 = Constant()
  n5 = If(
   A=Min(A=ComponentMask(None=@n3.Color) {g=true}, B=ComponentMask(None=@n3.Color) {b=true})
   B=SquareRoot(None=ScalarParameter() {defaultValue=0.001, group=Cartoon, parameterName=Subsurface Mask})
   "A > B"=@n4
   "A == B"=Constant() {r=1}
   "A < B"=@n4
  ) {equalsThreshold=0.0005}
  n6 = Floor(None=SceneTexture() {sceneTextureId=PPI_Opacity}.Color) {desc="for effect quality low (SSColor is disabled)"}
  n7 = Constant()
  n8 = If(
   A=ComponentMask(None=SceneTexture() {sceneTextureId=PPI_ShadingModelColor}.Color) {b=true, g=true, r=true}
   B=Constant3Vector() {constant=<rgba 0.7 1 1 0>}
   "A > B"=@n7
   "A == B"=Constant() {r=1}
   "A < B"=@n7
  ) {equalsThreshold=0.001}
  n9 = Constant() {r=1}
  n10 = NamedRerouteDeclaration(None=ShadingPathSwitch(Default=@n5, Deferred=QualitySwitch(Default=@n5, Low=@n6, High=@n8, Medium=@n8, Epic=@n5), Forward=@n9, Mobile=@n9)) {name=Subsurface Mask, nodeColor=<rgba 0.5469 0 1 1>}
  n11 = DynamicParameter() {paramNames [PosX,PosY,PosZ,Radius]}
  MaterialExpressionAppend3Vector_0 = MaterialXAppend3Vector!(A=@n11.PosX, B=@n11.PosX, C=@n11.PosX)
  n12 = Power(Base=SphereMask(
    A=MaterialFunctionCall() {materialFunction=&WorldPositionBehindTranslucency}.Result
    B=@MaterialExpressionAppend3Vector_0
    Radius=@n11.Radius
   ) {attenuationRadius=100, hardnessPercent=0})
  n13 = Constant()
  ShadingPathSwitch(Default=Multiply(A=@n1.A, B=Multiply(A=@n10, B=@n12)), Forward=@n13, Mobile=@n13)
 ]
}
