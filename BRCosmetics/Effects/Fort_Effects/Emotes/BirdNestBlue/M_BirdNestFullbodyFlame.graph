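# Translucent, unlit, two-sided Niagara sprite material that draws a full-body flame effect for the BirdNestBlue emote.
# A panning noise texture distorts center-scaled sprite UVs that feed sphere masks, shaping a soft flame silhouette.
# Noise and plasma pattern sample screen-aligned coordinates centered on the actor position, independent of sprite UVs.
# Opacity combines the masks, plasma, depth fade, particle alpha and a dynamic AlphaBase term; emissive tints a color parameter by particle color.
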Material /BRCosmetics/Effects/Fort_Effects/Emotes/BirdNestBlue/M_BirdNestFullbodyFlame
graph {
 properties {
  allowTranslucentCustomDepthWrites=true
  bUsedWithNiagaraSprites=true
  blendMode=BLEND_Translucent
  opacityMaskClipValue=0.001
  pixelDepthOffsetMode=PDOM_Legacy
  shadingModel=MSM_Unlit
  twoSided=true
  usedShadingModels=Unlit
 }
 nodes [
  n1 = VectorParameter() {parameterName=TotalUVScale, defaultValue=<rgba 0.25 0.25 0 1>}
  n2 = TextureObjectParameter() {
   group=NoiseA
   parameterName=NoiseA_Texture
   samplerType=SAMPLERTYPE_Normal
   sortPriority=0
   texture=&T_Weap_Distort_N_02
  }
  n3 = VectorParameter() {group=NoiseA, parameterName=NoiseA_UVControl, channelNames {a=VPan, b=UPan, g=VScale, r=UScale}, defaultValue=<rgba 22 15 0 0.1>}
  n4 = MaterialFunctionCall!(Center Point World Position=ActorPositionWS()) {materialFunction=/Engine/Functions/Engine_MaterialFunctions02/UVs/WorldSpaceAlignedScreenCoordinates}
  n5 = VectorParameter() {
   group=NoiseA
   parameterName=NoiseA_NoiseControl
   sortPriority=60
   channelNames {a=VStrength, b=UStrength, g=VOffset, r=UOffset}
   defaultValue=<rgba 0 0 0 1.5>
  }
  n6 = Add(
   A=MaterialFunctionCall(UVs=TextureCoordinate(), Texture Scale=AppendVector(A=@n1.R, B=@n1.R)) {materialFunction=&ScaleUVsByCenter}.UVs
   B=MaterialFunctionCall!(
    NoiseValue=MaterialFunctionCall(Texture=@n2, UVControl_RGB=@n3, UVControl_A=@n3, "[Override]UVInput"=@n4.Offset and Scaled UVS) {materialFunction=&FXCommon_TextureSample}.RG
    Control_RGB=@n5
    Control_A=@n5
   ) {materialFunction=/Game/Effects/Niagara/MasterMaterials/Functions/Common/FXCommon_NoiseControl}.NoiseValue
  )
  n7 = Constant() {r=0.5}
  n8 = SphereMask(A=@n6, B=@n7) {attenuationRadius=0.5, hardnessPercent=20}
  n9 = DepthFade() {fadeDistanceDefault=5}
  n10 = ParticleColor()
  n11 = Saturate(None=SphereMask(A=@n6, B=@n7) {attenuationRadius=0.5})
  n12 = Multiply(A=Multiply(A=Power(Base=@n9), B=@n10.A), B=Multiply(A=@n11) {constB=0.1})
  n13 = TextureObjectParameter() {
   group=ParttenTexture
   parameterName=ParttenTexture_Texture
   samplerType=SAMPLERTYPE_LinearGrayscale
   sortPriority=0
   texture=/Game/Effects/Fort_Effects/Emotes/UltraLight/Texture/T_Plasma_A_B_Greyscale
  }
  n14 = VectorParameter() {
   group=ParttenTexture
   parameterName=ParttenTexture_UVControl
   channelNames {a=VPan, b=UPan, g=VScale, r=UScale}
   defaultValue=<rgba 22.5 4 0 0.5>
  }
  n15 = MaterialFunctionCall(
   Texture=@n13
   UVControl_RGB=@n14
   UVControl_A=@n14
   "[Override]UVInput"=MaterialFunctionCall(Center Point World Position=ActorPositionWS()) {materialFunction=&WorldSpaceAlignedScreenCoordinates}.Offset and Scaled UVS
  ) {materialFunction=&FXCommon_TextureSample}
  Add(
   A=Multiply(A=Power(Base=@n8), B=Multiply(A=@n12, B=@n15.R))
   B=Multiply(A=@n11, B=DynamicParameter() {defaultValue=<rgba 0.001 1 1 1>, paramNames [AlphaBase,Param2,Param3,Param4]}.AlphaBase) {constB=0.001}
  ) {constB=0.001}
  n16 = VectorParameter() {parameterName=Color, defaultValue=<rgba 0 2 4 0.2>}
  MaterialFunctionCall(Emissive=Divide(A=Multiply(A=@n16.RGB, B=@n10.RGB))) {materialFunction=&MF_FX_EmissiveAdjustments}
  OneMinus(None=@n8)
 ]
}
