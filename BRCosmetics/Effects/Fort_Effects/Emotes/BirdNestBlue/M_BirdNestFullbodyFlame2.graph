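# Translucent, unlit, two-sided Niagara particle material for a full-body flame in the BirdNestBlue emote.
# UVs are noise-distorted, mirrored, bent by rotation and panned using dynamic-parameter time, falling back to Time.
# An energy-edge texture, contrasted and optionally multiplied by a screen-aligned plasma pattern, drives depth-faded opacity under tinted particle color.
# Low quality, and Medium on forward or mobile, switch to an aura texture with inner glow and hard stroke.
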Material /BRCosmetics/Effects/Fort_Effects/Emotes/BirdNestBlue/M_BirdNestFullbodyFlame2
graph {
 properties {
  allowTranslucentCustomDepthWrites=true
  bUsedWithNiagaraMeshParticles=true
  bUsedWithNiagaraSprites=true
  blendMode=BLEND_Translucent
  opacityMaskClipValue=0.001
  pixelDepthOffsetMode=PDOM_Legacy
  shadingModel=MSM_Unlit
  twoSided=true
  usedShadingModels=Unlit
 }
 nodes [
  n1 = ParticleColor()
  n2 = Multiply(A=VectorParameter() {parameterName=Color, defaultValue=<rgba 0 0.5949 1 0.2>}.RGB, B=@n1.RGB)
  n3 = ScalarParameter() {defaultValue=5, parameterName=LowPlatform_ScaleColor}
  n4 = Constant3Vector() {constant=<rgba 0 0 1 0>}
  n5 = TextureCoordinate()
  n6 = VectorParameter() {parameterName=TotalUVScaleAndOffset, defaultValue=<rgba 1 0.7143 0 0>}
  n7 = AppendVector(A=@n6.R, B=@n6.R)
  n8 = MaterialFunctionCall(Float2=TextureCoordinate()) {materialFunction=&BreakOutFloat2Components}
  n9 = VectorParameter() {
   group=NoiseMask
   parameterName=NoiseMask_EdgeMask_InvertControl
   channelNames {a="", b=Power, g=Multiply, r=Offset}
   defaultValue=<rgba 0 1 1 1>
  }
  n10 = VectorParameter() {group=NoiseMask, parameterName=NoiseMask_EdgeMask_Control, channelNames {a="", b=Power, g=Multiply, r=Offset}, defaultValue=<rgba 0 1 1 1>}
  n11 = TextureObjectParameter() {
   group=TotalNoise
   parameterName=TotalNoise_Texture
   samplerType=SAMPLERTYPE_Normal
   sortPriority=0
   texture=&T_Weap_Distort_N_02
  }
  n12 = VectorParameter() {group=TotalNoise, parameterName=TotalNoise_UVControl, channelNames {a=VPan, b=UPan, g=VScale, r=UScale}, defaultValue=<rgba 0.25 1.5 0 0.75>}
  n13 = MaterialFunctionCall(Texture=@n11, UVControl_RGB=@n12, UVControl_A=@n12, "[Override]TimeInput"=NamedRerouteUsage()) {materialFunction=&FXCommon_TextureSample}
  n14 = VectorParameter() {
   group=TotalNoise
   parameterName=TotalNoise_NoiseControl
   sortPriority=60
   channelNames {a=VStrength, b=UStrength, g=VOffset, r=UOffset}
   defaultValue=<rgba 0 0 0.2 0>
  }
  n15 = Add(
   A=Add(
    A=MaterialFunctionCall(UVs=@n5, Texture Scale=@n7) {materialFunction=&ScaleUVsByCenter}.UVs
    B=AppendVector(A=@n6.B, B=@n6.B)
   )
   B=Multiply(
    A=MaterialFunctionCall!(UorV=@n8.G, Invert_EdgeControl=@n9, NoInvert_EdgeControl=@n10) {materialFunction=/Game/Effects/Niagara/MasterMaterials/Functions/Common/FXCommon_EdgeMask_Ex}.Mask
    B=MaterialFunctionCall!(NoiseValue=@n13.RG, Control_RGB=@n14, Control_A=@n14) {materialFunction=/Game/Effects/Niagara/MasterMaterials/Functions/Common/FXCommon_NoiseControl}.NoiseValue
   )
  )
  n16 = MaterialFunctionCall(Float2=@n15) {materialFunction=&BreakOutFloat2Components}
  n17 = Multiply(A=Multiply(A=Subtract(A=@n16.G) {constB=0.5}) {constB=2}, B=ScalarParameter() {defaultValue=-0.204, parameterName=MainUVBend})
  n18 = VectorParameter() {parameterName=MainUVControl, defaultValue=<rgba 1.2784 2 0 0>}
  n19 = Add(A=Divide(A=@n18.RGB), B=VectorParameter() {parameterName=MainUVPivot, defaultValue=<rgba 0 0.2 0 1>}.RGB)
  n20 = MaterialFunctionCall(Float2=@n15) {materialFunction=&BreakOutFloat2Components}
  n21 = Multiply(A=Abs(None=Add(A=@n20.R) {constB=-0.5})) {constB=2}
  n22 = AppendVector(
   A=MaterialFunctionCall!(UVInput=AppendVector(A=@n21, B=@n20.G), TimeInput=NamedRerouteUsage(), Control_RGB=@n18.RGB, Control_A=@n18.RGB) {materialFunction=/Game/Effects/Niagara/MasterMaterials/Functions/Common/FXCommon_QuickUVControl}.UV
   B=Constant()
  )
  n23 = NamedRerouteUsage()
  n24 = VectorParameter() {parameterName=MainUVPannel, defaultValue=<rgba 0 2 0 1>}
  n25 = VectorParameter() {parameterName=MainUVOffset, defaultValue=<rgba -0.212 7.1226 0 1>}
  n26 = AppendVector(A=@n25.R, B=@n25.R)
  n27 = Add(
   A=Add(A=RotateAboutAxis(NormalizedRotationAxis=@n4, RotationAngle=@n17, PivotPoint=@n19, Position=@n22), B=@n22)
   B=Add(A=Multiply(A=@n23, B=@n24.RGB), B=AppendVector(A=@n26, B=@n25.B))
  )
  n28 = ComponentMask(None=@n27) {g=true, r=true}
  n29 = TextureSampleParameter2D(UVs=@n28) {parameterName=Param_Low, samplerType=SAMPLERTYPE_Normal, texture=&/T_Aura_Lowplatform}
  n30 = Saturate(None=ConstantBiasScale(None=@n29.R))
  n31 = MaterialFunctionCall!(TexCoord=Subtract(A=@n27) {constB=-0.01}, ClampChannel=Constant2Vector() {r=1}) {materialFunction=/Game/Effects/Fort_Effects/MaterialFunctions/MF_FX_TilingMethod}
  n32 = VectorParameter() {parameterName=LowPlatFormRefine, defaultValue=<rgba 1 1 1 0.5>}
  n33 = Saturate(None=MaterialFunctionCall(Texture=Multiply(A=@n30, B=@n31.Result), RGB=@n32.RGB, A=@n32.RGB) {materialFunction=&MF_FX_Refine}.Result)
  n34 = Saturate(None=ConstantBiasScale(None=@n29.G))
  n35 = Saturate(None=Subtract(A=Multiply(A=@n34, B=@n31.Result), B=ScalarParameter() {defaultValue=0.2, parameterName=LowPlatform_StrokeSize}))
  n36 = LinearInterpolate(A=LinearInterpolate(A=@n2, B=Multiply(A=@n2, B=@n3), Alpha=@n33), Alpha=Saturate(None=Multiply(A=@n35) {constB=999})) {constAlpha=0, constB=0}
  n37 = ShadingPathSwitch(Default=@n2, Forward=@n36, Mobile=@n36)
  MaterialFunctionCall(Emissive=Divide(A=QualitySwitch(Default=@n2, Low=@n36, Medium=@n37))) {materialFunction=&MF_FX_EmissiveAdjustments}
  Multiply(A=@n5, B=@n7)
  n38 = Constant() {r=1}
  n39 = DynamicParameter() {parameterIndex=3, defaultValue=<rgba 0 1 1 1>, paramNames [TimeInput,Param2,Param3,Param4]}
  n40 = MaterialFunctionCall(Float2=@n27) {materialFunction=&BreakOutFloat2Components}
  n41 = Constant()
  n42 = Constant() {r=1}
  n43 = TextureSampleParameter2D(UVs=@n28) {
   parameterName=Param
   samplerType=SAMPLERTYPE_Grayscale
   texture=&/T_BirdNestBlue_EnergyEdge
  }
  n44 = Saturate(None=Add(A=If(A=@n40.R, "A > B"=@n41, "A == B"=@n41, "A < B"=@n42) {constB=0.01}, B=Multiply(A=@n43.R, B=@n31.Result)))
  n45 = MaterialFunctionCall(In=@n44, Contrast=ScalarParameter() {defaultValue=3, parameterName=CheapContrast}) {materialFunction=&CheapContrast}
  n46 = StaticSwitchParameter(True=Saturate(None=Subtract(A=@n44, B=@n45.Result)), False=@n45.Result) {parameterName=ContrasToReduce}
  n47 = TextureObjectParameter() {
   group=ParttenTexture
   parameterName=ParttenTexture_Texture
   samplerType=SAMPLERTYPE_LinearGrayscale
   sortPriority=0
   texture=/Game/Effects/Fort_Effects/Emotes/UltraLight/Texture/T_Plasma_A_B_Greyscale
  }
  n48 = VectorParameter() {
   group=ParttenTexture
   parameterName=ParttenTexture_UVControl
   channelNames {a=VPan, b=UPan, g=VScale, r=UScale}
   defaultValue=<rgba 22.5 4 0 0.5>
  }
  n49 = NamedRerouteUsage()
  n50 = MaterialFunctionCall!(Center Point World Position=ActorPositionWS()) {materialFunction=/Engine/Functions/Engine_MaterialFunctions02/UVs/WorldSpaceAlignedScreenCoordinates}
  n51 = Multiply(
   A=Multiply(A=@n46, B=ScalarParameter() {defaultValue=0.1, parameterName=ScaleOpacity})
   B=StaticSwitchParameter(
    True=Constant() {r=1}
    False=MaterialFunctionCall(Texture=@n47, UVControl_RGB=@n48, UVControl_A=@n48, "[Override]TimeInput"=@n49, "[Override]UVInput"=@n50.Offset and Scaled UVS) {materialFunction=&FXCommon_TextureSample}.R
   ) {parameterName=DisableTexture}
  )
  n52 = Multiply(A=@n35) {constB=500}
  n53 = Multiply(A=@n33, B=ScalarParameter() {defaultValue=1, parameterName=LowPlatform_ScaleInnerGlowStr})
  n54 = Multiply(A=Add(A=Saturate(None=@n52), B=Add(A=@n53, B=@n51)), B=ScalarParameter() {defaultValue=2.5, parameterName=LowPlatformTotalOpacityScale})
  n55 = Multiply(A=@n1.A, B=QualitySwitch(Default=@n51, Low=@n54, Medium=ShadingPathSwitch(Default=@n51, Forward=@n54, Mobile=@n54)))
  n56 = Power(Base=DepthFade(FadeDistance=ScalarParameter() {defaultValue=50, parameterName=DepthFade}))
  n57 = Multiply(A=@n46) {constB=0.001}
  n58 = Constant()
  n59 = ShadingPathSwitch(Default=@n57, Forward=@n58, Mobile=@n58)
  Multiply(
   A=QualitySwitch(Default=@n38, Low=@n39.Param2, Medium=ShadingPathSwitch(Default=@n38, Forward=@n39.Param2, Mobile=@n39.Param2))
   B=Add(A=Multiply(A=@n55, B=@n56), B=QualitySwitch(Default=@n57, Low=@n58, Medium=@n59)) {constB=0.001}
  )
  n60 = DynamicParameter() {parameterIndex=3, defaultValue=<rgba 0 1 1 1>, paramNames [TimeInput,Param2,Param3,Param4]}
  n61 = Time() {bOverride_Period=true, period=6666}
  NamedRerouteDeclaration(None=If(A=@n60.TimeInput, B=Constant(), "A > B"=@n60.TimeInput, "A == B"=@n61, "A < B"=@n61)) {name=DPTime, nodeColor=<rgba 0 1 0.6975 1>}
 ]
}
