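# Translucent unlit Niagara particle material for the BirdNestBlue emote, drawing a top-view full-body flame as a tinted, noise-edged disc.
# Radial Perlin noise distorts UVs feeding sphere masks that form the disc, falloff and thin rim.
# Opacity combines the masks with a screen-aligned plasma pattern around the actor position and particle alpha.
# Low quality, and Medium on forward or mobile, add brightened falloff, dark outline and dynamic-parameter opacity.
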Material /BRCosmetics/Effects/Fort_Effects/Emotes/BirdNestBlue/M_BirdNestFullbodyFlameTopView
graph {
 properties {
  allowTranslucentCustomDepthWrites=true
  bUsedWithNiagaraMeshParticles=true
  bUsedWithNiagaraSprites=true
  blendMode=BLEND_Translucent
  opacityMaskClipValue=0
  pixelDepthOffsetMode=PDOM_Legacy
  shadingModel=MSM_Unlit
  twoSided=true
  usedShadingModels=Unlit
 }
 nodes [
  n1 = ParticleColor()
  n2 = Multiply(A=VectorParameter() {parameterName=Color, defaultValue=<rgba 0 0.5949 1 0.2>}.RGB, B=@n1.RGB)
  n3 = ScalarParameter() {defaultValue=5, parameterName=awaw}
  n4 = TextureObjectParameter() {
   group=NoiseA
   parameterName=NoiseA_Texture
   samplerType=SAMPLERTYPE_Normal
   sortPriority=0
   texture=/Game/Effects/Fort_Effects/Emotes/BirdNestBlue/T_LowResSimplePerlin
  }
  n5 = VectorParameter() {group=NoiseA, parameterName=NoiseA_UVControl, channelNames {a=VPan, b=UPan, g=VScale, r=UScale}, defaultValue=<rgba 1 0.2 0.2 -3>}
  n6 = MaterialFunctionCall() {materialFunction=&VectorToRadialValue}
  n7 = VectorParameter() {
   group=NoiseA
   parameterName=NoiseA_NoiseControl
   sortPriority=60
   channelNames {a=VStrength, b=UStrength, g=VOffset, r=UOffset}
   defaultValue=<rgba 0 0 0.1 0.1>
  }
  n8 = Add(
   A=TextureCoordinate()
   B=MaterialFunctionCall!(
    NoiseValue=MaterialFunctionCall!(
     Texture=@n4
     "[Override]DDXInput"
     "[Override]DDYInput"
     AdditionalUVOffset
     UVControl_RGB=@n5
     UVControl_A=@n5
     "[Override]TimeInput"
     "[Override]UVInput"=@n6.Radial Coordinates
    ) {materialFunction=/Game/Effects/Niagara/MasterMaterials/Functions/Common/FXCommon_TextureSample_DDXDDY}.RG
    Control_RGB=@n7
    Control_A=@n7
   ) {materialFunction=/Game/Effects/Niagara/MasterMaterials/Functions/Common/FXCommon_NoiseControl}.NoiseValue
  )
  n9 = Constant() {r=0.5}
  n10 = Power(
   Base=OneMinus(None=SphereMask(A=@n8, B=@n9) {attenuationRadius=0.35, hardnessPercent=1})
   Exp=ScalarParameter() {defaultValue=2, parameterName=ssa}
  )
  n11 = SphereMask(A=@n8, B=@n9) {attenuationRadius=0.35}
  n12 = Saturate(None=Subtract(A=SphereMask(A=@n8, B=@n9) {attenuationRadius=0.36}, B=@n11))
  n13 = LinearInterpolate(A=LinearInterpolate(A=@n2, B=Multiply(A=@n2, B=@n3), Alpha=@n10), Alpha=Multiply(A=@n12) {constB=1000}) {constB=0}
  n14 = ShadingPathSwitch(Default=@n2, Forward=@n13, Mobile=@n13)
  MaterialFunctionCall(Emissive=Divide(A=QualitySwitch(Default=@n2, Low=@n13, Medium=@n14))) {materialFunction=&MF_FX_EmissiveAdjustments}
  n15 = Constant() {r=1}
  n16 = DynamicParameter() {parameterIndex=3, defaultValue=<rgba 0 1 1 1>, paramNames [TimeInput,Param2,Param3,Param4]}
  n17 = TextureObjectParameter() {
   group=ParttenTexture
   parameterName=ParttenTexture_Texture
   samplerType=SAMPLERTYPE_LinearGrayscale
   sortPriority=0
   texture=/Game/Effects/Fort_Effects/Emotes/UltraLight/Texture/T_Plasma_A_B_Greyscale
  }
  n18 = VectorParameter() {
   group=ParttenTexture
   parameterName=ParttenTexture_UVControl
   channelNames {a=VPan, b=UPan, g=VScale, r=UScale}
   defaultValue=<rgba 22.5 4 0 0.5>
  }
  n19 = ActorPositionWS()
  n20 = Multiply(
   A=Multiply(A=@n11, B=ScalarParameter() {defaultValue=0.1, parameterName=SO})
   B=MaterialFunctionCall(
    Texture=@n17
    UVControl_RGB=@n18
    UVControl_A=@n18
    "[Override]UVInput"=MaterialFunctionCall!(Center Point World Position=@n19) {materialFunction=/Engine/Functions/Engine_MaterialFunctions02/UVs/WorldSpaceAlignedScreenCoordinates}.Offset and Scaled UVS
   ) {materialFunction=&FXCommon_TextureSample}.R
  )
  n21 = VectorParameter() {parameterName=LowPlatFormRefine, defaultValue=<rgba 10 1 0.05 0.5>}
  n22 = Multiply(
   A=Saturate(None=MaterialFunctionCall(Texture=@n10, RGB=@n21.RGB, A=@n21.RGB) {materialFunction=&MF_FX_Refine}.Result)
   B=ScalarParameter() {defaultValue=0.5, parameterName=a}
  )
  n23 = Add(A=Saturate(None=Multiply(A=@n12) {constB=500}), B=Add(A=@n20, B=Multiply(A=@n22, B=@n11)))
  n24 = QualitySwitch(Default=@n20, Low=@n23, Medium=ShadingPathSwitch(Default=@n20, Forward=@n23, Mobile=@n23))
  Multiply(
   A=QualitySwitch(Default=@n15, Low=@n16.Param2, Medium=ShadingPathSwitch(Default=@n15, Forward=@n16.Param2, Mobile=@n16.Param2))
   B=Add(A=Multiply(A=@n1.A, B=@n24), B=Multiply(A=@n11) {constB=0.001}) {constB=0.001}
  )
 ]
}
