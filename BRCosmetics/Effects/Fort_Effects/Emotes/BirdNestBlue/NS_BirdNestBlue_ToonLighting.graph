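# Niagara system for the BirdNestBlue emote that, by its name, provides toon-style lighting at the character's hands.
# A burst spawns two camera-offset sprites, using spawn groups to track the left and right hand bones.
# Emitter-age curves with random scaling blend color from white to bright blue, with time-of-day alpha.
# Dynamic material parameters pass particle world position and a curve-driven scalar, likely light radius, to two sprite renderers.
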NiagaraSystem /BRCosmetics/Effects/Fort_Effects/Emotes/BirdNestBlue/NS_BirdNestBlue_ToonLighting
dependencies {
 usedDataInterfaces [&NiagaraDataInterfaceSkeletalMesh, &NiagaraDataInterfaceCurve]
 usedDynamicInputs []
 usedModules [&SystemState]
}
emitters {Empty {
  emitterUpdateScript [
   &EmitterState(Inactive Response=::NewEnumerator1, Life Cycle Mode=::NewEnumerator1, Loop Behavior=::NewEnumerator1, UseLoopDelay=false)
   &SpawnBurst_Instantaneous(Spawn Count=2, Use Spawn Probability=false, UseLoopCountLimit=false)
  ]
  particleSpawnScript [
   &FN_SetSpawnGroup(Number of Spawn Groups=2)
   &InitializeParticle(
    Color=<rgba 5.5 14.6667 27.5 1>
    Lifetime=5
    Sprite Size Mode=::NewEnumerator3
    Uniform Sprite Size=100
    UsePositionOffset=true
    Write Material Random=true
   )
   &CameraOffset(Camera Offset Amount=50)
  ]
  particleUpdateScript [
   &ParticleState()
   &SkeletalMeshLocation(
    Bone / Socket Index=&SpawnGroup()
    Bone Sampling Mode=::NewEnumerator4
    MaskSpawnGroup=false
    Sampled Position Offset=<0 0 20>
    Sampled Position Offset Coordinate Space:ENiagaraSkelMeshTransforms=::NewEnumerator1
    Skeletal Mesh=DataInterface() {propertyValues {
      DefaultMesh=null
      ExcludeBoneName=root
      OptionalComponentClass=null
      PreviewMesh=&M_Med_Soldier_04_Preview
      SkinningMode=SkinOnTheFly
      SoftSourceActor=null
      SourceMode=Default
      UvSetIndex=0
      WholeMeshLOD=-1
      bExcludeBone=false
      bReadDeformedGeometry=true
      bRequireCurrentFrameData=true
      ComponentTags []
      FilteredBones [hand_l,hand_r]
      FilteredSockets []
      MeshUserParameter {parameter {name=None, typeDefHandle {registeredTypeIndex=1}, varData []}}
      SamplingRegions []
     }}
   )
   /SetVariables(Particles.Color=&MakeLinearColorFromVectorAndFloat(
     "Vector (RGB)"=&Lerp_Vector(
      A=<1 1 1>
      B=&MakeVectorFromLinearColorRGB(Color=$Particles.Initial.Color)
      Alpha=&FloatFromCurve(
       FloatCurve=DataInterface() {propertyValues {Curve {keys [
           {interpMode=RCIM_Cubic, leaveTangent=0.5575, tangentMode=RCTM_Break, time=0.5, value=0}
           {arriveTangent=0.2435, interpMode=RCIM_Cubic, leaveTangent=10.1128, tangentMode=RCTM_Break, time=1.75, value=0.5}
           {interpMode=RCIM_Cubic, leaveTangent=-0.0028, tangentMode=RCTM_Break, time=1.8, value=1}
           {interpMode=RCIM_Cubic, tangentMode=RCTM_Break, time=4.15, value=0.6}
           {interpMode=RCIM_Cubic, tangentMode=RCTM_Break, time=4.2, value=1}
           <at 4.4 0>
          ]}}}
       CurveIndex=$Emitter.Age
       Scale Curve=&RandomRangeFloat(Minimum=0.5, Maximum=1, Evaluation Type=::NewEnumerator1)
      )
     )
     "Float (Alpha)"=&FNTimeOfDayScalar(Night Scale=0.75, Morning Scale=1, Evening Scale=1)
    )) {isSetParameters=true}
   &DynamicMaterialParameters(
    Index 0 Param 1=&MakeFloatFromVector(
     Vector=&TransformPosition(Position=$Particles.Position, Source Space=::Local, Destination Space=::World)
     Channel=::NewEnumerator0
    )
    Index 0 Param 2=&MakeFloatFromVector(
     Vector=&TransformPosition(Position=$Particles.Position, Source Space=::Local, Destination Space=::World)
     Channel=::NewEnumerator1
    )
    Index 0 Param 3=&MakeFloatFromVector(
     Vector=&TransformPosition(Position=$Particles.Position, Source Space=::Local, Destination Space=::World)
     Channel=::NewEnumerator2
    )
    Index 0 Param 4=&Lerp_Float(
     A=50
     B=35
     Alpha=&FloatFromCurve(
      FloatCurve=DataInterface() {propertyValues {Curve {keys [
          {interpMode=RCIM_Cubic, leaveTangent=0.0118, tangentMode=RCTM_Break, time=0.5, value=0}
          {arriveTangent=0.0161, interpMode=RCIM_Cubic, leaveTangent=10.1128, tangentMode=RCTM_Break, time=1.75, value=0}
          {interpMode=RCIM_Cubic, leaveTangent=-0.0028, tangentMode=RCTM_Break, time=1.8, value=1}
          {interpMode=RCIM_Cubic, tangentMode=RCTM_Break, time=4.2, value=1}
          <at 4.4 0>
         ]}}}
      CurveIndex=$Emitter.Age
      Scale Curve=&RandomRangeFloat(Minimum=0.5, Maximum=1, Evaluation Type=::NewEnumerator1)
     )
    )
    Module.Param0WriteEnabled=true
    Module.Param1WriteEnabled=true
    Module.Param2WriteEnabled=true
    Module.Param3WriteEnabled=true
   )
  ]
  rendererClasses [&NiagaraSpriteRendererProperties]
  renderers [
   {rendererClass=&NiagaraSpriteRendererProperties}
   {rendererClass=&NiagaraSpriteRendererProperties, rendererIndex=1}
  ]
 }}
summary {
 systemName=NS_BirdNestBlue_ToonLighting
 emitters [{emitterName=Empty, rendererClasses [&NiagaraSpriteRendererProperties]}]
 userVariables []
}
