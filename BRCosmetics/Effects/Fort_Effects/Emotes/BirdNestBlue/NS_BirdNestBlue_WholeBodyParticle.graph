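# Niagara system for the BirdNestBlue emote's whole-body aura and glow; all four emitters are disabled here.
# Two timed aura sprites, one upright and one facing up, crossfade by System.CameraAngleDot to suit the view angle.
# WholeBodyGlow pins eleven sprites to User.TargetSM bones or sockets by spawn group, while Bounds adds a large sprite.
# Age curves drive fades, aura scale pops and material parameters, with time-of-day dimming.
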NiagaraSystem /BRCosmetics/Effects/Fort_Effects/Emotes/BirdNestBlue/NS_BirdNestBlue_WholeBodyParticle
dependencies {
 usedDataInterfaces [&NiagaraDataInterfaceCurve, &NiagaraDataInterfaceCamera]
 usedDynamicInputs [&Abs_Float, &DotProduct, &ReturnCameraVector]
 usedModules [
  &SystemState
  &self:SystemSpawnScript.SystemScriptSource.SystemScriptGraph.NiagaraNodeAssignment_0.SetVariables_#1
  &self:SystemSpawnScript.SystemScriptSource.SystemScriptGraph.NiagaraNodeAssignment_1.SetVariables_#2
 ]
}
emitters {
 Aura_Body {
  bEnabled=false
  emitterUpdateScript [
   &EmitterState(
    Inactive Response=::NewEnumerator1
    Life Cycle Mode=::NewEnumerator1
    Loop Behavior=::NewEnumerator1
    Loop Duration Mode=::NewEnumerator1
    UseLoopDelay=false
   )
   &SpawnBurst_Instantaneous(Use Spawn Probability=false, UseLoopCountLimit=false)
  ]
  particleSpawnScript [
   &InitializeParticle(
    Color=<rgba 1 1 1 0.4>
    Lifetime=5
    Position Offset=<0 0 -10>
    Sprite Size=<165 373>
    Sprite Size Mode=::NewEnumerator4
    UsePositionOffset=true
    Write Material Random=true
   )
   &InitializeParticle(
    Lifetime=5
    Position Offset=<0 0 -10>
    Sprite Size=<150 350>
    Sprite Size Mode=::NewEnumerator4
    UsePositionOffset=true
    Write Material Random=true
   ) {enabled=false, module=InitializeParticle001}
   &SpriteFacingAndAlignment(Sprite Alignment=<0 0 1>, Sprite Facing=<0 0 0>, WriteAlignment=true, WriteFacing=false)
   &CameraOffset(Camera Offset Amount=75)
  ]
  particleUpdateScript [
   &ParticleState()
   &Timeline(
    Play Rate=&FloatFromCurve(
     FloatCurve=DataInterface() {propertyValues {Curve {keys [
         <at 0 1>
         {interpMode=RCIM_Cubic, tangentMode=RCTM_Break, time=1.8, value=1}
         {interpMode=RCIM_Cubic, leaveTangent=-0.1423, tangentMode=RCTM_Break, time=1.9, value=1.7}
         {arriveTangent=-0.627, interpMode=RCIM_Cubic, tangentMode=RCTM_Break, time=3.1, value=1}
        ]}}}
     CurveIndex=$Emitter.Age
    )
    Rewind Mode:ENiagara_TimelineMode=::NewEnumerator1
    Timeline End=10000000
   )
   &ScaleColor(
    Scale Alpha=&FloatFromCurve(
     FloatCurve=DataInterface() {propertyValues {Curve {keys [
         <at 0.5 0>
         {arriveTangent=0.065, interpMode=RCIM_Cubic, leaveTangent=10.1128, tangentMode=RCTM_Break, time=1.75, value=0.1}
         {interpMode=RCIM_Cubic, leaveTangent=-0.0028, tangentMode=RCTM_Break, time=2, value=1}
         {interpMode=RCIM_Cubic, tangentMode=RCTM_Break, time=4.2, value=1}
         <at 4.8 0>
        ]}}}
     CurveIndex=$Particles.Age
    )
    ScaleA=true
    ScaleRGB=true
   )
   &ScaleSpriteSize(
    Uniform Curve Index=$Particles.Age
    Uniform Curve Scale=&FloatFromCurve(FloatCurve=DataInterface() {propertyValues {Curve {keys [<at 0.7 1>,<at 1 0>]}}}, CurveIndex=$System.CameraAngleDot)
    Uniform Curve Sprite Scale=DataInterface() {propertyValues {Curve {keys [
        {arriveTangent=0.0123, interpMode=RCIM_Cubic, leaveTangent=8.5926, tangentMode=RCTM_Break, time=1.75, value=0}
        {interpMode=RCIM_Cubic, leaveTangent=0.5002, tangentMode=RCTM_Break, time=2, value=1.1}
        {arriveTangent=-0.125, interpMode=RCIM_Cubic, leaveTangent=-0.125, time=2.6, value=0.9}
        {arriveTangent=0.0313, interpMode=RCIM_Cubic, leaveTangent=0.0313, time=3.2, value=0.95}
        {interpMode=RCIM_Cubic, tangentMode=RCTM_Break, time=4.2, value=0.95}
       ]}}}
   )
   &ScaleColor(Scale RGB=&VectorFromFloat(Value=$System.TODColor), ScaleA=true, ScaleRGB=true) {module=ScaleColor001}
   /SetVariables(Particles.Color2=&MultiplyLinearColor(
     Linear Color=&MultiplyLinearColor(
      Linear Color=$Particles.Color
      Scale Factor=&MakeVector4(
       X=1
       Y=1
       Z=1
       W=&FloatFromCurve(
        FloatCurve=DataInterface() {propertyValues {Curve {keys [
            {interpMode=RCIM_Cubic, leaveTangent=-0.0028, tangentMode=RCTM_Break, time=2, value=1}
            {interpMode=RCIM_Cubic, tangentMode=RCTM_Break, time=4, value=1}
            <at 4.8 0>
           ]}}}
        CurveIndex=$Particles.Age
       )
      )
     )
     Scale Factor=&MakeVector4(X=1, Y=1, Z=1, W=&Waveform("Waveform [1]"=::NewEnumerator0, "[1] Frequency"=5, "[1] Amplitude Min/Max"=<1 2>))
    )) {isSetParameters=true}
   &DynamicMaterialParameters(
    Index 3 Param 1=$Particles.Timeline.Time
    Index 3 Param 2=&FloatFromCurve(FloatCurve=DataInterface() {propertyValues {Curve {keys [<at 0.7 1>,<at 0.8 0>]}}}, CurveIndex=$System.CameraAngleDot)
    Module.Param0WriteEnabled3=true
    Module.Param1WriteEnabled3=true
    Module.Param2WriteEnabled3=true
    Module.Param3WriteEnabled3=true
    Write Parameter Index 0=false
    Write Parameter Index 3=true
   )
  ]
  rendererClasses [&NiagaraSpriteRendererProperties]
  renderers [
   {rendererClass=&NiagaraSpriteRendererProperties}
   {rendererClass=&NiagaraSpriteRendererProperties, rendererIndex=1}
  ]
 }
 Aura_Body001 {
  bEnabled=false
  emitterUpdateScript [
   &EmitterState(
    Inactive Response=::NewEnumerator1
    Life Cycle Mode=::NewEnumerator1
    Loop Behavior=::NewEnumerator1
    Loop Duration Mode=::NewEnumerator1
    UseLoopDelay=false
   )
   &SpawnBurst_Instantaneous(Use Spawn Probability=false, UseLoopCountLimit=false)
  ]
  particleSpawnScript [
   &InitializeParticle(
    Color=<rgba 1 1 1 0.85>
    Lifetime=5
    Sprite Size=<175 175>
    Sprite Size Mode=::NewEnumerator4
    UsePositionOffset=true
    Write Material Random=true
   )
   &SpriteFacingAndAlignment(Sprite Alignment=<0 0 0>, Sprite Facing=<0 0 1>, WriteAlignment=false, WriteFacing=true)
   &CameraOffset(Camera Offset Amount=75)
  ]
  particleUpdateScript [
   &ParticleState()
   &ScaleColor(
    Scale Alpha=&FloatFromCurve(
     FloatCurve=DataInterface() {propertyValues {Curve {keys [
         <at 0.5 0>
         {arriveTangent=0.065, interpMode=RCIM_Cubic, leaveTangent=10.1128, tangentMode=RCTM_Break, time=1.75, value=0.1}
         {interpMode=RCIM_Cubic, leaveTangent=-0.0028, tangentMode=RCTM_Break, time=2, value=1}
         {interpMode=RCIM_Cubic, tangentMode=RCTM_Break, time=4.2, value=1}
         <at 4.8 0>
        ]}}}
     CurveIndex=$Particles.Age
    )
    ScaleA=true
    ScaleRGB=true
   )
   &ScaleSpriteSize(
    Uniform Curve Index=$Particles.Age
    Uniform Curve Scale=&FloatFromCurve(FloatCurve=DataInterface() {propertyValues {Curve {keys [<at 0.4 0>,<at 0.7 1>]}}}, CurveIndex=$System.CameraAngleDot)
    Uniform Curve Sprite Scale=DataInterface() {propertyValues {Curve {keys [
        {arriveTangent=0.0123, interpMode=RCIM_Cubic, leaveTangent=8.5926, tangentMode=RCTM_Break, time=1.75, value=0}
        {interpMode=RCIM_Cubic, leaveTangent=0.5002, tangentMode=RCTM_Break, time=2, value=1.1}
        {arriveTangent=-0.25, interpMode=RCIM_Cubic, leaveTangent=-0.25, time=2.3, value=0.9}
        {arriveTangent=0.0263, interpMode=RCIM_Cubic, leaveTangent=0.0263, time=2.6, value=0.95}
        {interpMode=RCIM_Cubic, tangentMode=RCTM_Break, time=4.2, value=0.95}
       ]}}}
   )
   &ScaleColor(Scale RGB=&VectorFromFloat(Value=$System.TODColor), ScaleA=true, ScaleRGB=true) {module=ScaleColor001}
   &DynamicMaterialParameters(
    Index 3 Param 2=&FloatFromCurve(FloatCurve=DataInterface() {propertyValues {Curve {keys [<at 0.6 0>,<at 0.7 1>]}}}, CurveIndex=$System.CameraAngleDot)
    Module.Param0WriteEnabled3=true
    Module.Param1WriteEnabled3=true
    Module.Param2WriteEnabled3=true
    Module.Param3WriteEnabled3=true
    Write Parameter Index 0=false
    Write Parameter Index 3=true
   )
  ]
  rendererClasses [&NiagaraSpriteRendererProperties]
  renderers [{rendererClass=&NiagaraSpriteRendererProperties}]
 }
 Bounds {
  bEnabled=false
  emitterUpdateScript [
   &EmitterState(Life Cycle Mode=::NewEnumerator1, Loop Behavior=::NewEnumerator1, Loop Duration=7.178, UseLoopDelay=false)
   &SpawnBurst_Instantaneous(Use Spawn Probability=false, UseLoopCountLimit=false)
  ]
  particleSpawnScript [&InitializeParticle(
    Lifetime=5
    Position Offset=<0 0 100>
    Sprite Size=&Multiply_Vector2D_ByFloat(Float=1.2, Vector2D=<210 280>)
    Sprite Size Mode=::NewEnumerator4
    UsePositionOffset=true
    Write Material Random=true
   )]
  particleUpdateScript [
   &ParticleState()
   &CameraOffset()
   &ScaleColor(
    Scale Alpha=&FloatFromCurve(
     FloatCurve=DataInterface() {propertyValues {Curve {keys [
         <at 0.5 0>
         {arriveTangent=0.065, interpMode=RCIM_Cubic, leaveTangent=5.4347, tangentMode=RCTM_Break, time=1.75, value=0}
         {interpMode=RCIM_Cubic, leaveTangent=-0.0028, tangentMode=RCTM_Break, time=2, value=1}
         {interpMode=RCIM_Cubic, tangentMode=RCTM_Break, time=4.2, value=1}
         <at 4.5 0>
        ]}}}
     CurveIndex=$Particles.Age
    )
    ScaleA=true
    ScaleRGB=true
   )
   &ScaleColor(
    Scale RGB=&VectorFromFloat(Value=&FNTimeOfDayScalar(Night Scale=0.3, Morning Scale=1, Evening Scale=1))
    ScaleA=true
    ScaleRGB=true
   ) {module=ScaleColor001}
   /SetVariables(Particles.Color2=&MultiplyLinearColor(
     Linear Color=$Particles.Color
     Scale Factor=&MakeVector4(
      X=1
      Y=1
      Z=1
      W=&FloatFromCurve(
       FloatCurve=DataInterface() {propertyValues {Curve {keys [
           <at 0.5 0>
           {arriveTangent=0.065, interpMode=RCIM_Cubic, leaveTangent=5.4347, tangentMode=RCTM_Break, time=1.75, value=0}
           {interpMode=RCIM_Cubic, leaveTangent=-0.0028, tangentMode=RCTM_Break, time=2, value=1}
           {interpMode=RCIM_Cubic, tangentMode=RCTM_Break, time=4.2, value=1}
           <at 4.5 0>
          ]}}}
       CurveIndex=$Particles.Age
      )
     )
    )) {isSetParameters=true}
  ]
  rendererClasses [&NiagaraSpriteRendererProperties]
  renderers [
   {rendererClass=&NiagaraSpriteRendererProperties}
   {rendererClass=&NiagaraSpriteRendererProperties, rendererIndex=1}
  ]
 }
 WholeBodyGlow {
  bEnabled=false
  emitterUpdateScript [
   &EmitterState(Life Cycle Mode=::NewEnumerator1, Loop Behavior=::NewEnumerator1, UseLoopDelay=false)
   &SpawnBurst_Instantaneous(Spawn Count=11, Use Spawn Probability=false, UseLoopCountLimit=false)
  ]
  particleSpawnScript [
   &FN_SetSpawnGroup(Number of Spawn Groups=11)
   &InitializeParticle(
    Color=<rgba 0.75 0.75 0.75 0.8>
    Lifetime=5
    Sprite Size Mode=::NewEnumerator1
    Uniform Sprite Size Max=150
    Uniform Sprite Size Min=100
    UsePositionOffset=false
    Write Material Random=true
   )
   &SpriteFacingAndAlignment(Sprite Alignment=<0 0 1>, Sprite Facing=<0 0 0>, WriteAlignment=true, WriteFacing=false) {enabled=false}
  ]
  particleUpdateScript [
   &ParticleState()
   &CameraOffset(Camera Offset Amount=50)
   &SkeletalMeshLocation(Bone / Socket Index=$Particles.SpawnGroup, Bone Sampling Mode=::NewEnumerator4, MaskSpawnGroup=false, Skeletal Mesh=$User.TargetSM)
   &ScaleColor(
    Scale Alpha=&FloatFromCurve(
     FloatCurve=DataInterface() {propertyValues {Curve {keys [
         <at 0.5 0>
         {arriveTangent=0.065, interpMode=RCIM_Cubic, leaveTangent=10.1128, tangentMode=RCTM_Break, time=1.75, value=0.4}
         {interpMode=RCIM_Cubic, leaveTangent=-0.0028, tangentMode=RCTM_Break, time=1.8, value=1}
         {interpMode=RCIM_Cubic, tangentMode=RCTM_Break, time=4.2, value=1}
         <at 4.8 0>
        ]}}}
     CurveIndex=$Particles.Age
    )
    ScaleA=true
    ScaleRGB=true
   )
   &ScaleSpriteSize(
    Uniform Curve Index=$Particles.Age
    Uniform Curve Sprite Scale=DataInterface() {propertyValues {Curve {keys [{interpMode=RCIM_Cubic, tangentMode=RCTM_Break, time=4.2, value=1}, <at 4.8 0>]}}}
   ) {enabled=false}
   &DynamicMaterialParameters(
    Index 0 Param 1=&FloatFromCurve(
     FloatCurve=DataInterface() {propertyValues {Curve {keys [
         <at 0 0>
         {interpMode=RCIM_Cubic, leaveTangent=-0.0028, tangentMode=RCTM_Break, time=1.8, value=1}
         {interpMode=RCIM_Cubic, tangentMode=RCTM_Break, time=4.2, value=1}
         <at 4.8 0.5>
        ]}}}
     CurveIndex=$Particles.Age
     Scale Curve=0.002
    )
    Module.Param0WriteEnabled=true
    Module.Param1WriteEnabled=true
    Module.Param2WriteEnabled=true
    Module.Param3WriteEnabled=true
   )
   &ScaleColor(
    Scale RGB=&VectorFromFloat(Value=&FNTimeOfDayScalar(Night Scale=0.4, Morning Scale=1, Evening Scale=1))
    ScaleA=true
    ScaleRGB=true
   ) {module=ScaleColor001}
  ]
  rendererClasses [&NiagaraSpriteRendererProperties]
  renderers [{rendererClass=&NiagaraSpriteRendererProperties}]
 }
}
summary {
 systemName=NS_BirdNestBlue_WholeBodyParticle
 emitters [
  {bEnabled=false, emitterName=WholeBodyGlow, rendererClasses [&NiagaraSpriteRendererProperties]}
  {bEnabled=false, emitterName=Bounds, rendererClasses [&NiagaraSpriteRendererProperties]}
  {bEnabled=false, emitterName=Aura_Body, rendererClasses [&NiagaraSpriteRendererProperties]}
  {bEnabled=false, emitterName=Aura_Body001, rendererClasses [&NiagaraSpriteRendererProperties]}
 ]
 userVariables [
  User.bIsFrontEnd {type=NiagaraBool}
  User.bIsFrontEndPreview {type=NiagaraBool}
  User.TargetSM {
   type=NiagaraDataInterfaceSkeletalMesh
   defaultValue {
    struct=NiagaraExt_VariableValue_DataInterface
    value {dataInterface=&self:NiagaraDataInterfaceSkeletalMesh_0}
   }
  }
 ]
}
